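# Looping ambient SoundCue for the clapping hologram in the Neon City POI, played under the ambient sound class.
# A mixer layers a looping hum/buzz and a looping UltraSam hologram sound, each with a one-second volume fade-in and small-hologram attenuation, with UltraSam weighted louder.
# A third branch with its own attenuation repeats a Luna commercial voice line after a random delay of roughly 12 to 16 seconds.
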SoundCue /Game/Sounds/Asteria/POI/NeonCity/Holograms/NeonCity_ClappingHologram_Loop_Cue
root=Mixer_1
nodes [
 {
  id=Attenuation_0
  children [Enveloper_1]
  properties {attenuationSettings=&/NeonCity_Hologram_Small_Att}
 }
 {
  id=Attenuation_2
  children [Looping_0]
  properties {attenuationSettings=&/NeonCity_ClappingHologram_Att}
 }
 {
  id=Attenuation_3
  children [Enveloper_2]
  properties {attenuationSettings=&NeonCity_Hologram_Small_Att}
 }
 {id=Delay_0, children [WavePlayer_0], properties {delayMax=16, delayMin=12}}
 {
  id=Enveloper_1
  children [WavePlayer_1]
  properties {pitchCurve {editorCurveData {keys [{time=0}]}}, volumeCurve {editorCurveData {keys [<at 0 0>,<at 1 1>]}}}
 }
 {
  id=Enveloper_2
  children [WavePlayer_2]
  properties {pitchCurve {editorCurveData {keys [{time=0}]}}, volumeCurve {editorCurveData {keys [<at 0 0>,<at 1 1>]}}}
 }
 {id=Looping_0, children [Delay_0]}
 {id=Mixer_1, children [Attenuation_0, Attenuation_3, Attenuation_2], properties {inputVolume [1,2,1]}}
 {id=WavePlayer_0, properties {soundWaveAssetPtr=/Game/Sounds/Asteria/POI/NeonCity/Spots/comm_vo_luna_042}}
 {id=WavePlayer_1, properties {bLooping=true, soundWaveAssetPtr=&/Hologram_Giant_Hum_Buzz_Loop_01}}
 {id=WavePlayer_2, properties {bLooping=true, soundWaveAssetPtr=&/NeonCity_Hologram_UltraSam_Loop}}
]
properties {
 bOverrideConcurrency=true
 maxDistance=12500
 soundClassObject=&Fort_Ambient_Athena
 volumeMultiplier=0.752
 attenuationOverrides {falloffDistance=2000}
 concurrencyOverrides {maxCount=1}
 concurrencySet [&Weapon_Drop_StopOldest]
}
