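# Sound cue for the dragon hologram's fire burst in the NeonCity POI, playing on the Athena ambient sound class.
# A mixer layers a close roar wave and a distant roar wave, each through its own attenuation asset, with the close layer boosted.
# The two attenuation layers let the perceived sound shift with listener distance.
# The cue's debug flag is left enabled.
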SoundCue /Game/Sounds/Asteria/POI/NeonCity/Holograms/NeonCity_Dragon_Hologram_Fire_Burst_Cue
root=Mixer_0
nodes [
 {
  id=Attenuation_0
  children [WavePlayer_4]
  properties {attenuationSettings=&/NeonCity_Dragon_Hologram_Close_Att}
 }
 {
  id=Attenuation_1
  children [WavePlayer_5]
  properties {attenuationSettings=&/NeonCity_Dragon_Hologram_Distant_Att}
 }
 {id=Mixer_0, children [Attenuation_0, Attenuation_1], properties {inputVolume [2.25,1]}}
 {id=WavePlayer_4, properties {soundWaveAssetPtr=&/NeonCity_Dragon_Hologram_Roar_Close}}
 {id=WavePlayer_5, properties {soundWaveAssetPtr=&/NeonCity_Dragon_Hologram_Roar_Distant}}
]
properties {
 bDebug=true
 maxDistance=20000
 soundClassObject=&Fort_Ambient_Athena
 volumeMultiplier=0.576
 attenuationOverrides {falloffDistance=512, attenuationShapeExtents {x=256}}
 concurrencyOverrides {maxCount=6}
}
