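# Sound cue for the roar and fire audio of the NeonCity dragon hologram, playing a start sound that leads into a sustained loop.
# A mixer plays the one-shot start alongside the looping wave, which is weighted louder.
# The loop's enveloper holds it silent briefly, then fades it in under the start, and a smoothed Doppler node shifts its pitch with relative motion.
# It uses dedicated attenuation and an ambient sound class.
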SoundCue /Game/Sounds/Asteria/POI/NeonCity/Holograms/NeonCity_Dragon_Hologram_Fire_Loop_Cue
root=Mixer_0
nodes [
 {id=Doppler_0, children [Enveloper_1], properties {bUseSmoothing=true, dopplerIntensity=1.3}}
 {
  id=Enveloper_1
  children [WavePlayer_2]
  properties {pitchCurve {editorCurveData {keys [{time=0}]}}, volumeCurve {editorCurveData {keys [<at 0 0 RCIM_Cubic>,<at 0.5 0>,<at 2 1>]}}}
 }
 {id=Mixer_0, children [WavePlayer_0,Doppler_0], properties {inputVolume [1,1.5]}}
 {id=WavePlayer_0, properties {soundWaveAssetPtr=&/NeonCity_Dragon_Hologram_Roar_Fire_Loop_Start}}
 {
  id=WavePlayer_2
  properties {bLooping=true, soundWaveAssetPtr=&/NeonCity_Dragon_Hologram_Roar_Fire_Loop}
 }
]
properties {
 attenuationSettings=&/NeonCity_Dragon_Hologram_Fire_Att
 maxDistance=21000
 soundClassObject=&Fort_Ambient_Athena
 volumeMultiplier=0.501
 attenuationOverrides {falloffDistance=2000}
 concurrencyOverrides {maxCount=1}
 concurrencySet [&Weapon_Drop_StopOldest]
}
