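# Sound cue that plays the looping ambient audio for the fish school hologram at the Neon City POI.
# A single looping wave player feeds the output directly, using the fish hologram attenuation asset and the Athena ambient sound class.
# Overridden concurrency limits it to one instance, so the assigned weapon-drop stop-oldest concurrency set is bypassed.
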SoundCue /Game/Sounds/Asteria/POI/NeonCity/Holograms/NeonCity_FishHologram_Loop_Cue
root=WavePlayer_3
nodes [{
  id=WavePlayer_3
  properties {bLooping=true, soundWaveAssetPtr=&/NeonCity_Hologram_Fish_School_Loop}
 }]
properties {
 attenuationSettings=&/NeonCity_Hologram_Fish_Att
 bOverrideConcurrency=true
 maxDistance=6000
 soundClassObject=&Fort_Ambient_Athena
 volumeMultiplier=0.927
 attenuationOverrides {falloffDistance=2000}
 concurrencyOverrides {maxCount=1}
 concurrencySet [&Weapon_Drop_StopOldest]
}
