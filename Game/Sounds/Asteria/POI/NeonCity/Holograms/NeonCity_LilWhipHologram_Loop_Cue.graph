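# Sound cue for the looping music of the Lil Whip hologram at the Neon City point of interest.
# A single looping wave player plays the hologram music loop directly as the cue's root.
# The cue uses the small hologram attenuation asset and the ambient sound class, with overridden concurrency capping simultaneous instances and a slightly reduced volume.
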SoundCue /Game/Sounds/Asteria/POI/NeonCity/Holograms/NeonCity_LilWhipHologram_Loop_Cue
root=WavePlayer_0
nodes [{
  id=WavePlayer_0
  properties {bLooping=true, soundWaveAssetPtr=&/NeonCity_LilWhipHologram_Music_Loop}
 }]
properties {
 attenuationSettings=&/NeonCity_Hologram_Small_Att
 bOverrideConcurrency=true
 maxDistance=3000
 priority=0.9
 soundClassObject=&Fort_Ambient_Athena
 volumeMultiplier=0.927
 attenuationOverrides {falloffDistance=512, attenuationShapeExtents {x=256}}
 concurrencyOverrides {maxCount=6}
}
