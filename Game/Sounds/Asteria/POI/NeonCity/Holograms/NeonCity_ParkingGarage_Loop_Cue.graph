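# Looping ambient sound cue for a parking garage hologram at the Neon City POI, in the outdoor ambience sound class.
# Picks at random among four equally weighted hologram loops borrowed from Keys_And_Locks foley, then passes it through a modulator and a small-hologram attenuation node.
# A concurrency override allows one instance, stopping the quietest, and virtualization is set to play when silent.
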SoundCue /Game/Sounds/Asteria/POI/NeonCity/Holograms/NeonCity_ParkingGarage_Loop_Cue
root=Attenuation_0
nodes [
 {
  id=Attenuation_0
  children [Modulator_0]
  properties {attenuationSettings=&/NeonCity_Hologram_Small_Att}
 }
 {id=Modulator_0, children [Random_0], properties {pitchMin=1}}
 {id=Random_0, children [WavePlayer_1,WavePlayer_2,WavePlayer_3,WavePlayer_4], properties {weights [1,1,1,1]}}
 {
  id=WavePlayer_1
  properties {bLooping=true, soundWaveAssetPtr=/Game/Sounds/Foley_Loot/Containers/Keys_And_Locks/Keys_And_Locks_HologramLoop_01}
 }
 {
  id=WavePlayer_2
  properties {bLooping=true, soundWaveAssetPtr=/Game/Sounds/Foley_Loot/Containers/Keys_And_Locks/Keys_And_Locks_HologramLoop_02}
 }
 {
  id=WavePlayer_3
  properties {bLooping=true, soundWaveAssetPtr=/Game/Sounds/Foley_Loot/Containers/Keys_And_Locks/Keys_And_Locks_HologramLoop_03}
 }
 {
  id=WavePlayer_4
  properties {bLooping=true, soundWaveAssetPtr=/Game/Sounds/Foley_Loot/Containers/Keys_And_Locks/Keys_And_Locks_HologramLoop_04}
 }
]
properties {
 bOverrideConcurrency=true
 maxDistance=3000
 soundClassObject=&Fort_Ambient_Outdoors_Athena
 virtualizationMode=PlayWhenSilent
 volumeMultiplier=1.002
 attenuationOverrides {
  bEnableOcclusion=true
  falloffDistance=925
  nonFocusVolumeAttenuation=0
  nonSpatializedRadiusStart=900
  occlusionLowPassFilterFrequency=2000
  occlusionTraceChannel=ECC_GameTraceChannel6
  occlusionVolumeAttenuation=0.7
  stereoSpread=1500
  attenuationShapeExtents {x=900}
 }
 concurrencyOverrides {bVolumeScaleCanRelease=true, maxCount=1, resolutionRule=StopQuietest}
 soundSubmixSends [{minSendLevel=0.5, soundSubmix=&Fortnite_Reverb_Default}]
}
