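# Looping ambient sound cue for the samurai hologram at the Neon City POI, mixed from three separately attenuated layers.
# A hum-buzz loop (medium attenuation) and an UltraSam loop (small attenuation) each pass through a volume fade-in envelope.
# A Fujita commercial voice-over uses large attenuation and repeats through a looping node with a fixed delay before each play.
# Overridden concurrency limits the cue to one instance, in the ambient sound class.
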SoundCue /Game/Sounds/Asteria/POI/NeonCity/Holograms/NeonCity_SamuraiHologram_Loop_Cue
root=Mixer_0
nodes [
 {
  id=Attenuation_0
  children [Enveloper_0]
  properties {attenuationSettings=&/NeonCity_Hologram_Medium_Att}
 }
 {
  id=Attenuation_1
  children [Looping_0]
  properties {attenuationSettings=&/NeonCity_Hologram_Large_Att}
 }
 {
  id=Attenuation_2
  children [Enveloper_1]
  properties {attenuationSettings=&/NeonCity_Hologram_Small_Att}
 }
 {id=Delay_0, children [WavePlayer_5], properties {delayMax=16, delayMin=16}}
 {
  id=Enveloper_0
  children [WavePlayer_0]
  properties {pitchCurve {editorCurveData {keys [{time=0}]}}, volumeCurve {editorCurveData {keys [<at 0 0>,<at 1 1>]}}}
 }
 {
  id=Enveloper_1
  children [WavePlayer_1]
  properties {pitchCurve {editorCurveData {keys [{time=0}]}}, volumeCurve {editorCurveData {keys [<at 0 0>,<at 1 1>]}}}
 }
 {id=Looping_0, children [Delay_0]}
 {id=Mixer_0, children [Attenuation_0, Attenuation_2, Attenuation_1], properties {inputVolume [2,4,1.25]}}
 {id=WavePlayer_0, properties {bLooping=true, soundWaveAssetPtr=&/Hologram_Giant_Hum_Buzz_Loop_01}}
 {id=WavePlayer_1, properties {bLooping=true, soundWaveAssetPtr=&/NeonCity_Hologram_UltraSam_Loop}}
 {id=WavePlayer_5, properties {soundWaveAssetPtr=/Game/Sounds/Asteria/POI/NeonCity/Spots/comm_vo_fujita_041}}
]
properties {
 bOverrideConcurrency=true
 maxDistance=18000
 soundClassObject=&Fort_Ambient_Athena
 volumeMultiplier=1.002
 attenuationOverrides {falloffDistance=2000}
 concurrencyOverrides {maxCount=1}
 concurrencySet [&Weapon_Drop_StopOldest]
}
