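# A looping ambient sound cue for the abstract-art waves holographic sign in the Neon City POI.
# A single looping wave player feeds the output directly, using a dedicated attenuation asset and the Athena ambient sound class.
# Concurrency is overridden to allow one instance at a time, though a Weapon_Drop_StopOldest concurrency set is also referenced.
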SoundCue /Game/Sounds/Asteria/POI/NeonCity/Holograms/NeonCity_Sign_Abstract_Art_Waves_Loop_Cue
root=WavePlayer_2
nodes [{
  id=WavePlayer_2
  properties {bLooping=true, soundWaveAssetPtr=&/NeonCity_Sign_Abstract_Art_Waves_Loop}
 }]
properties {
 attenuationSettings=&/NeonCity_Sign_Abstract_Art_Wave_Att
 bOverrideConcurrency=true
 maxDistance=2000
 soundClassObject=&Fort_Ambient_Athena
 volumeMultiplier=0.902
 attenuationOverrides {falloffDistance=2000}
 concurrencyOverrides {maxCount=1}
 concurrencySet [&Weapon_Drop_StopOldest]
}
