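# Sound cue for the deactivation sound of hologram figures in the Neon City POI.
# A Random node picks one of three equally weighted deactivate waves and feeds a root Modulator that holds volume constant, leaving pitch variation at defaults.
# It plays on the Athena ambient sound class, references the small hologram activate attenuation asset and overrides concurrency to cap simultaneous instances.
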SoundCue /Game/Sounds/Asteria/POI/NeonCity/Holograms/NeonCity_StatMeshHolograms_Off_Cue
root=Modulator_0
nodes [
 {id=Modulator_0, children [Random_0], properties {volumeMax=1, volumeMin=1}}
 {id=Random_0, children [WavePlayer_1,WavePlayer_2,WavePlayer_3], properties {bShouldExcludeFromBranchCulling=true, weights [1,1,1]}}
 {id=WavePlayer_1, properties {soundWaveAssetPtr=&/Hologram_Figure_Deactivate_01}}
 {id=WavePlayer_2, properties {soundWaveAssetPtr=&/Hologram_Figure_Deactivate_02}}
 {id=WavePlayer_3, properties {soundWaveAssetPtr=&/Hologram_Figure_Deactivate_03}}
]
properties {
 attenuationSettings=&/Hologram_Figure_Small_Activate_Att
 bOverrideConcurrency=true
 maxDistance=3000
 priority=0.9
 soundClassObject=&Fort_Ambient_Athena
 volumeMultiplier=0.752
 attenuationOverrides {falloffDistance=512, attenuationShapeExtents {x=256}}
 concurrencyOverrides {maxCount=6}
}
