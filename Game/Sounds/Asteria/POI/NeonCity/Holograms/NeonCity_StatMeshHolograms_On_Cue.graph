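# Sound cue that plays the activation sound for static mesh holograms at the NeonCity POI.
# A Random node picks one of three equally weighted hologram figure activation waves and feeds it through a Modulator.
# It uses a small activation attenuation asset and the ambient Athena sound class at reduced volume, with a concurrency override limiting simultaneous instances.
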SoundCue /Game/Sounds/Asteria/POI/NeonCity/Holograms/NeonCity_StatMeshHolograms_On_Cue
root=Modulator_0
nodes [
 {id=Modulator_0, children [Random_0]}
 {id=Random_0, children [WavePlayer_0,WavePlayer_2,WavePlayer_3], properties {bShouldExcludeFromBranchCulling=true, weights [1,1,1]}}
 {id=WavePlayer_0, properties {soundWaveAssetPtr=&/Hologram_Figure_Activate_01}}
 {id=WavePlayer_2, properties {soundWaveAssetPtr=&/Hologram_Figure_Activate_02}}
 {id=WavePlayer_3, properties {soundWaveAssetPtr=&/Hologram_Figure_Activate_03}}
]
properties {
 attenuationSettings=&/Hologram_Figure_Small_Activate_Att
 bOverrideConcurrency=true
 maxDistance=3000
 soundClassObject=&Fort_Ambient_Athena
 volumeMultiplier=0.401
 attenuationOverrides {falloffDistance=512, attenuationShapeExtents {x=256}}
 concurrencyOverrides {maxCount=6}
}
