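# Looping ambient sound cue for the wall-standing hologram at the NeonCity point of interest, in the Athena ambient sound class.
# A mixer layers a looping hum-and-buzz wave, weighted louder, with a looping music wave.
# Each layer has its own attenuation node, a medium hologram falloff for the hum and a wall-standing hologram falloff for the music.
# A concurrency override limits the cue to one playing instance.
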SoundCue /Game/Sounds/Asteria/POI/NeonCity/Holograms/NeonCity_WallStandingHologram_Loop_Cue
root=Mixer_0
nodes [
 {
  id=Attenuation_0
  children [WavePlayer_0]
  properties {attenuationSettings=&/NeonCity_Hologram_Medium_Att}
 }
 {
  id=Attenuation_1
  children [WavePlayer_5]
  properties {attenuationSettings=&/NeonCity_WallStandingHologram_Att}
 }
 {id=Mixer_0, children [Attenuation_0, Attenuation_1], properties {inputVolume [2,1]}}
 {id=WavePlayer_0, properties {bLooping=true, soundWaveAssetPtr=&/Hologram_Giant_Hum_Buzz_Loop_01}}
 {id=WavePlayer_5, properties {bLooping=true, soundWaveAssetPtr=&/Hologram_Giant_Music_Loop_01}}
]
properties {
 bOverrideConcurrency=true
 maxDistance=13900
 soundClassObject=&Fort_Ambient_Athena
 volumeMultiplier=0.752
 attenuationOverrides {falloffDistance=2000}
 concurrencyOverrides {maxCount=1}
 concurrencySet [&Weapon_Drop_StopOldest]
}
