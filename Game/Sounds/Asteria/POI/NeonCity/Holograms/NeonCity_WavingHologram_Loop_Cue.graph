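# Looping ambient sound cue for a waving hologram at the NeonCity POI, limited to one instance by a concurrency override.
# It mixes three layers, each with its own attenuation: an UltraSam loop (small), a hum-buzz loop (tiny) and a voice spot (large).
# The two loops are pitched up by modulators, and the UltraSam loop fades in through an enveloper.
# The voice spot repeats after random delays via looping and delay nodes.
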SoundCue /Game/Sounds/Asteria/POI/NeonCity/Holograms/NeonCity_WavingHologram_Loop_Cue
root=Mixer_0
nodes [
 {
  id=Attenuation_0
  children [Modulator_1]
  properties {attenuationSettings=&/NeonCity_Hologram_Small_Att}
 }
 {
  id=Attenuation_1
  children [Looping_0]
  properties {attenuationSettings=&/NeonCity_Hologram_Large_Att}
 }
 {
  id=Attenuation_2
  children [Modulator_0]
  properties {attenuationSettings=&/NeonCity_Hologram_Tiny_Att}
 }
 {id=Delay_0, children [WavePlayer_0], properties {delayMax=20, delayMin=12}}
 {
  id=Enveloper_0
  children [WavePlayer_5]
  properties {pitchCurve {editorCurveData {keys [{time=0}]}}, volumeCurve {editorCurveData {keys [<at 0 0>,<at 1 1>]}}}
 }
 {id=Looping_0, children [Delay_0]}
 {id=Mixer_0, children [Attenuation_0, Attenuation_1, Attenuation_2], properties {inputVolume [1,2,2.5]}}
 {id=Modulator_0, children [WavePlayer_1], properties {pitchMax=3, pitchMin=3, volumeMax=1, volumeMin=1}}
 {id=Modulator_1, children [Enveloper_0], properties {pitchMax=3, pitchMin=3, volumeMax=1, volumeMin=1}}
 {id=WavePlayer_0, properties {soundWaveAssetPtr=/Game/Sounds/Asteria/POI/NeonCity/Spots/comm_vo_luna_041}}
 {id=WavePlayer_1, properties {bLooping=true, soundWaveAssetPtr=&/Hologram_Giant_Hum_Buzz_Loop_01}}
 {id=WavePlayer_5, properties {bLooping=true, soundWaveAssetPtr=&/NeonCity_Hologram_UltraSam_Loop}}
]
properties {
 attenuationSettings=&/NeonCity_Hologram_Medium_Att
 bOverrideConcurrency=true
 maxDistance=18000
 soundClassObject=&Fort_Ambient_Athena
 volumeMultiplier=0.626
 attenuationOverrides {falloffDistance=2000}
 concurrencyOverrides {maxCount=1}
 concurrencySet [&Weapon_Drop_StopOldest]
}
